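# Niagara toon-shader example system with one mesh emitter, Snot, demonstrating an anime outline look.
# Emitter spawn modules set outline width and toon highlight parameters, and two mesh renderers draw the particles.
# A burst spawns randomly scaled meshes at shape locations, with looping particle lifetimes and per-particle material random values.
# ScaleColor takes a cyan tint from FNTimeOfDayColor, which supplies different brightness for night, morning, day and evening.
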NiagaraSystem /Game/NeverCook/FX/Niagara_FX_Reuse_Examples/ToonShader/NS_AnimeOutline_FXlnst
dependencies {
 usedDataInterfaces [&NiagaraDataInterfacePlatformSet]
 usedDynamicInputs []
 usedModules [&SystemState]
}
emitters {Snot {
  emitterSpawnScript [
   /Game/Effects/Niagara/MasterMaterials/ToonShader/Template/NM_OutlineParameters(OutlineWidth=5)
   /Game/Effects/Niagara/MasterMaterials/ToonShader/Template/NM_ToonShaderParameters(HighLight_Size=0.15, HighLight_Softness=1)
  ]
  emitterUpdateScript [
   &EmitterState(Life Cycle Mode=::NewEnumerator1, Loop Behavior=::NewEnumerator1, Loop Delay=0.05, Loop Duration Mode=::NewEnumerator1, UseLoopDelay=false)
   &SpawnBurst_Instantaneous(Spawn Count=10, Use Spawn Probability=false, UseLoopCountLimit=false)
  ]
  particleSpawnScript [
   &InitializeParticle(
    Lifetime=4
    Mesh Scale Mode=::NewEnumerator1
    Mesh Uniform Scale Max=0.2
    Mesh Uniform Scale Min=0.1
    UsePositionOffset=false
    Write Material Random=true
   )
   &InitialMeshOrientation(Mesh Orientation Mode=::NewEnumerator2, Rotation=<0 -0.05 0>, Use Rotation Vector=true)
   &ShapeLocation(MaskSpawnGroup=false, Override Seed=false, UseOwnerScale=true)
  ]
  particleUpdateScript [
   &ParticleState(
    Kill Particles When Lifetime Has Elapsed=false
    Let Infinitely Lived Particles Die When Emitter Deactivates=true
    Loop Particles Lifetime=true
   )
   &SolveForcesAndVelocity(Clamp Velocity=false, Limit Acceleration=false)
   &ScaleColor(
    Scale Mode=::NewEnumerator1
    Scale RGBA=&MakeVector4_FromLinearColor(LinearColor=&FNTimeOfDayColor(
      Night Color=<rgba 0.1904 0.4895 0.5625 1>
      Morning Color=<rgba 0.3385 0.8703 1 1>
      Day Color=<rgba 0.5078 1.3054 1.5 1>
      Evening Color=<rgba 0.6771 1.7405 2 1>
     ))
    ScaleRGBA=true
   )
  ]
  rendererClasses [&NiagaraMeshRendererProperties]
  renderers [
   {rendererClass=&NiagaraMeshRendererProperties, rendererIndex=0}
   {rendererClass=&NiagaraMeshRendererProperties, rendererIndex=1}
  ]
 }}
summary {
 systemName=NS_AnimeOutline_FXlnst
 emitters [{emitterName=Snot, rendererClasses [&NiagaraMeshRendererProperties]}]
 userVariables []
}
